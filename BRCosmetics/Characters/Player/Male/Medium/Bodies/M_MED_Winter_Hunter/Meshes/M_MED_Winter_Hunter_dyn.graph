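# Physics asset for the dynamic bones of the M_MED_Winter_Hunter male body, simulating a grappling hook and rope hung from the pelvis.
# Capsule bodies cover two clips, four rope segments, the hook, the pelvis and the left thigh.
# The first clip rotates freely on the pelvis, leading through angle-limited joints to the rope and hook.
# The chain forks at the first inside rope segment, with the middle branch ending in the hook.
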PhysicsAsset /BRCosmetics/Characters/Player/Male/Medium/Bodies/M_MED_Winter_Hunter/Meshes/M_MED_Winter_Hunter_dyn
bodies {
 dyn_grappleHookClip_01 [{shapeType=Capsule, length=1, radius=1}]
 dyn_grappleHookClip_02 [{shapeType=Capsule, length=1.9729, radius=1.3635, center=<1 0 0>, rotation=<rot 0 0 90>}]
 dyn_hook [{shapeType=Capsule, length=5.2376, radius=3.9522, center=<11.6336 0.2299 0.4916>, rotation=<rot 0 91.3429 -87.6573>}]
 dyn_ropeInside_01 [{shapeType=Capsule, length=3, radius=2, center=<5 0 2>, rotation=<rot 0 0 90>}]
 dyn_ropeInside_02 [{shapeType=Capsule, length=3, radius=2, center=<5 0 1.5>, rotation=<rot 0 0 90>}]
 dyn_ropeMIddle_01 [{shapeType=Capsule, length=3, radius=2, center=<3 0 0.5>, rotation=<rot 0 0 90>}]
 dyn_ropeMIddle_02 [{shapeType=Capsule, length=3, radius=2.2, center=<3 0 -0.6>, rotation=<rot 0 0 90>}]
 pelvis [{shapeType=Capsule, length=8.8335, radius=10.6426, center=<-0.7898 -1.0017 0>}]
 thigh_l [{shapeType=Capsule, length=24.948, radius=8.8979, center=<-18.9969 0.0004 0.1553>, rotation=<rot 87.1189 -179.9999 180>}]
}
constraints [
 {bone1Name=thigh_l, bone2Name=pelvis, swing1LimitDegrees=45, swing2LimitDegrees=45, twistLimitDegrees=45}
 {
  bone1Name=dyn_grappleHookClip_01
  bone2Name=pelvis
  swing1LimitDegrees=45
  swing1Motion=Free
  swing2LimitDegrees=45
  swing2Motion=Free
  twistLimitDegrees=45
  twistMotion=Free
 }
 {bone1Name=dyn_grappleHookClip_02, bone2Name=dyn_grappleHookClip_01, swing1LimitDegrees=75, swing2LimitDegrees=60, twistLimitDegrees=15}
 {bone1Name=dyn_ropeInside_01, bone2Name=dyn_grappleHookClip_02, swing1LimitDegrees=45, swing2LimitDegrees=60, twistLimitDegrees=15}
 {bone1Name=dyn_ropeInside_02, bone2Name=dyn_ropeInside_01, swing1LimitDegrees=45, swing2LimitDegrees=60, twistLimitDegrees=15}
 {bone1Name=dyn_ropeMIddle_01, bone2Name=dyn_ropeInside_01, swing1LimitDegrees=45, swing2LimitDegrees=60, twistLimitDegrees=15}
 {bone1Name=dyn_ropeMIddle_02, bone2Name=dyn_ropeMIddle_01, swing1LimitDegrees=45, swing2LimitDegrees=60, twistLimitDegrees=15}
 {bone1Name=dyn_hook, bone2Name=dyn_ropeMIddle_02, swing1LimitDegrees=45, swing2LimitDegrees=60, twistLimitDegrees=15}
]
