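# Animation Blueprint for the Winter Hunter body cosmetic, adding secondary motion to its dagger, grapple hook clip and chain bones.
# It copies the attached parent mesh's pose, then constrains clip and dagger bones to weighted pelvis, spine and thigh bones, with a Look At orienting the dagger.
# An additive clip offset switches on while crouching, via an alpha updated each frame.
# RigidBody simulation in pelvis space and spring-based AnimDynamics on the left chains finish the pose.
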AnimBlueprint /BRCosmetics/Characters/Player/Male/Medium/Bodies/M_MED_Winter_Hunter/Meshes/M_MED_Winter_Hunter_AnimBP
target_skeleton=&/M_MED_Winter_Hunter_Skeleton
graphs [
 AnimGraph {nodes [
   n1 = "Constraint - dyn_grappleHookClip_01 (PPPP)"!(
    ComponentPose=Local To Component(LocalPose=Copy Pose From Mesh() {bUseAttachedParent=true})
    Alpha=1
    ConstraintWeights_0=0.63
    ConstraintWeights_1:float=0.273
    ConstraintWeights_2:float=0.085
    ConstraintWeights_3:float=0.012
   ) -> (Pose) {
    alphaBoolBlend {blendOption=Linear}
    boneToModify {boneName=dyn_grappleHookClip_01}
    constraintSetup [{targetBone {boneName=pelvis}}, {targetBone {boneName=spine_01}}, {targetBone {boneName=spine_02}}, {targetBone {boneName=spine_03}}]
    constraintWeights [1,1,1,1]
   }
   n2 = "Constraint - dyn_dagger (PP)"!(ComponentPose=@n1, Alpha=1, ConstraintWeights_0=0.23, ConstraintWeights_1:float=0.77) -> (Pose) {
    alphaBoolBlend {blendOption=Linear}
    boneToModify {boneName=dyn_dagger}
    constraintSetup [{targetBone {boneName=thigh_r}}, {targetBone {boneName=thigh_twist_01_r}}]
    constraintWeights [1,1]
   }
   n3 = "Constraint - dyn_dagger_Aim (PP)"!(ComponentPose=@n2, Alpha=1, ConstraintWeights_0=0.4, ConstraintWeights_1:float=0.6) -> (Pose) {
    alphaBoolBlend {blendOption=Linear}
    boneToModify {boneName=dyn_dagger_Aim}
    constraintSetup [{targetBone {boneName=thigh_r}}, {targetBone {boneName=thigh_twist_01_r}}]
    constraintWeights [1,1]
   }
   n4 = Look At - Bone: dyn_dagger(ComponentPose=@n3, Alpha=1) {
    alphaBoolBlend {blendOption=Linear}
    boneToModify {boneName=dyn_dagger}
    lookAtLocation {x=0}
    lookAtTarget {boneReference {boneName=dyn_dagger_Aim}}
    lookAt_Axis {axis=<1 0>}
   }
   n5 = "Transform (Modify) Bone - Bone: dyn_grappleHookClip_01"(ComponentPose=@n4, Alpha=Get crouch alpha() -> (crouch alpha:double), Translation="0,-2.5,1", Rotation="0,-20,0", Scale="1,1,1") {
    rotationMode=BMM_Additive
    rotationSpace=BCS_BoneSpace
    translationMode=BMM_Additive
    translationSpace=BCS_BoneSpace
    alphaBoolBlend {blendOption=Linear}
    boneToModify {boneName=dyn_grappleHookClip_01}
    rotation {yaw=-20}
    translation {y=-2.5, z=1}
   }
   n6 = RigidBody(ComponentPose=@n5) {
    bClampLinearTranslationLimitToRefPose=true
    bOverrideWorldGravity=true
    overridePhysicsAsset=&/M_MED_Winter_Hunter_dyn
    simulationSpace=BaseBoneSpace
    alphaBoolBlend {blendOption=Linear}
    baseBoneRef {boneName=pelvis}
    componentAppliedLinearAccClamp=<2000 1000 2000>
    componentLinearAccScale=<0.2 0.1 0.4>
    componentLinearVelScale=<0.2 0.1 0.4>
    overrideWorldGravity {z=-980}
   }
   n7 = Anim Dynamics - Bone: dyn_chains_l(ComponentPose=@n6, Alpha=1, ExternalForce="0,0,0") {
    angularDampingOverride=0.998
    bAngularSpring=true
    bGravityOverrideInSimSpace=true
    bOverrideAngularDamping=true
    bOverrideLinearDamping=true
    bUseGravityOverride=true
    linearDampingOverride=0.9
    alphaBoolBlend {blendOption=Linear}
    boundBone {boneName=dyn_chains_l}
    componentAppliedLinearAccClamp=<3000 3000 3000>
    componentLinearAccScale=<0.35 0.35 0.35>
    componentLinearVelScale=<0.3 0.3 0.3>
    gravityOverride {z=-980}
    physicsBodyDefinitions [{
      sphereCollisionRadius=0
      boundBone {boneName=dyn_chains_l}
      boxExtents=<8 8 10>
      constraintSetup {angularLimitsMax=<30 15 60>, angularLimitsMin=<-30 -15 -60>, angularTarget=<0 0 0>, linearAxesMax=<0.5 0.5 2>, linearAxesMin=<-0.5 -0.5 -2>}
      localJointOffset=<1 -1 -0.5>
     }]
   }
   Output Pose(Result=Component To Local(ComponentPose=@n7)) {layerGroup=None}
  ]}
 EventGraph {nodes [
   Event BlueprintUpdateAnimation()
   > Set crouch alpha(crouch alpha:double="To Float (Boolean)"(InBool=Get bIsCrouching())) -> (Output_Get:double)
  ]}
]
variables [crouch alpha {type=number}]
