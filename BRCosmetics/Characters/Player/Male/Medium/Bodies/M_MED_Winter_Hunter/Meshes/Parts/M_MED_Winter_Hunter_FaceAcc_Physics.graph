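# Physics asset for the face accessory part of the male medium Winter Hunter player cosmetic, defining physics for its hat bone.
# It holds two capsule bodies, a small one on dyn_hat and a larger one around the head.
# A single constraint joins dyn_hat to head with angular limits on both swing axes and a very tight twist limit, so the hat can tilt but barely twist.
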PhysicsAsset /BRCosmetics/Characters/Player/Male/Medium/Bodies/M_MED_Winter_Hunter/Meshes/Parts/M_MED_Winter_Hunter_FaceAcc_Physics
bodies {
 dyn_hat [{shapeType=Capsule, length=6, radius=2.15, center=<5 0 0>, rotation=<rot 0 0 90>}]
 head [{shapeType=Capsule, length=3.9456, radius=9.2225, center=<3.9267 1.8474 0.0465>, rotation=<rot 90 0 0>}]
}
constraints [{bone1Name=dyn_hat, bone2Name=head, swing1LimitDegrees=30, swing2LimitDegrees=60, twistLimitDegrees=2}]
